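# Animation Blueprint for the Mainframe backpack that hides the pack while a matching pickaxe is equipped outside skydiving.
# Weapon-change and skydive begin/end events check for the MainframeMale meta tag on the equipped pickaxe, then set hiderScale to 1 or 0.
# That value is the alpha of a Modify Bone node that scales the reference pose's Root bone to zero.
# Each check also broadcasts the hidden state through the HideShield dispatcher.
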AnimBlueprint /BRCosmetics/Accessories/FORT_Backpacks/Backpack_M_MED_Mainframe/Meshes/M_MED_Mainframe_Pack_AnimBP
target_skeleton=&/M_MED_Mainframe_Pack_Skeleton
graphs [
 AnimGraph {nodes [
   n1 = "Transform (Modify) Bone - Bone: Root"(
    ComponentPose=Local To Component(LocalPose=Local Space Ref Pose())
    Alpha=Get hiderScale() -> (hiderScale:double)
    Translation="0,0,0"
    Rotation="0,0,0"
    Scale="0,0,0"
   ) {
    rotationMode=BMM_Additive
    scaleMode=BMM_Replace
    scaleSpace=BCS_BoneSpace
    alphaBoolBlend {blendOption=Linear}
    boneToModify {boneName=Root}
    scale=<0 0 0>
   }
   Output Pose(Result=Component To Local(ComponentPose=@n1)) {layerGroup=None}
  ]}
 EventGraph {nodes [
   n1 = Event OnBeginSkydiving()
   n2 = Make Array("[0]"=(TagName="Cosmetics.MetaTags.MainframeMale"))
   n3 = Get bIsSkyDiving()
   n4 = Branch(
    execute=[@n1.then, @n5.then, @n6.then]
    Condition=AND Boolean(A=NOT Boolean(A=@n3), B=PickaxeItemDefinitionHasAnyMetaTag(Tags=@n2, bRequiresEquipped=true))
   )
   > Set hiderScale(hiderScale:double=1) -> (Output_Get:double)
   > Call Hide Shield!(execute:Exec, self:M MED Mainframe Pack Anim BP Object Reference, isHidden:bool=true) -> (then:Exec) {is=title K2Node_CallDelegate}
   Set hiderScale(execute=@n4.else, hiderScale=0)
   > Call Hide Shield()
   n5 = Event OnWeaponChanged()
   n6 = Event OnEndSkydiving()
   TryGetPawnOwner()
   Event BlueprintUpdateAnimation()
  ]}
 HideShield {nodes [HideShield!() -> (then, isHidden:bool=~) {is=title K2Node_FunctionEntry}]}
]
variables [
 hiderScale {type=number}
 HideShield {default=()}
]
